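# Looping daytime exterior ambience cue for the Castle Doom POI courtyard, built from the Morning wave assets and routed to the outdoor ambient sound class with ambient attenuation.
# A quality-level switch selects the quad recording for its first two slots and the stereo one for the last two.
# Each looping wave runs through two nested, endlessly looping envelopers that slowly modulate pitch and volume over a thirty-second cycle to vary the bed.
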SoundCue /Game/Sounds/Helios/POI/CastleDoom/Ambience_DCastle_Courtyard_Day_Ext_Loop_Cue
root=QualityLevel_0
nodes [
 {
  id=Enveloper_0
  children [WavePlayer_1]
  properties {
   bLoop=true
   bLoopIndefinitely=true
   loopEnd=30
   pitchCurve {editorCurveData {keys [{interpMode=RCIM_Cubic, time=0}, <at 15 0.7 RCIM_Cubic>, {interpMode=RCIM_Cubic, time=30}]}}
   volumeCurve {editorCurveData {keys [<at 0 1>]}}
  }
 }
 {
  id=Enveloper_1
  children [WavePlayer_3]
  properties {
   bLoop=true
   bLoopIndefinitely=true
   loopEnd=30
   pitchCurve {editorCurveData {keys [{interpMode=RCIM_Cubic, time=0}, <at 15 0.7 RCIM_Cubic>, {interpMode=RCIM_Cubic, time=30}]}}
   volumeCurve {editorCurveData {keys [<at 0 1>]}}
  }
 }
 {
  id=Enveloper_2
  children [Enveloper_1]
  properties {
   bLoop=true
   bLoopIndefinitely=true
   loopEnd=30
   pitchCurve {editorCurveData {keys [{time=0}]}}
   volumeCurve {editorCurveData {keys [
      {arriveTangent=-0.01, interpMode=RCIM_Cubic, time=0, value=1}
      <at 15 1.1 RCIM_Cubic>
      {arriveTangent=-0.0004, interpMode=RCIM_Cubic, leaveTangent=-0.0004, tangentMode=RCTM_User, time=35.1185, value=1}
      {interpMode=RCIM_Cubic, leaveTangent=0.01, time=60, value=1}
     ]}}
  }
 }
 {
  id=Enveloper_3
  children [Enveloper_0]
  properties {
   bLoop=true
   bLoopIndefinitely=true
   loopEnd=30
   pitchCurve {editorCurveData {keys [{time=0}]}}
   volumeCurve {editorCurveData {keys [
      {arriveTangent=-0.01, interpMode=RCIM_Cubic, time=0, value=1}
      <at 15 1.1 RCIM_Cubic>
      {arriveTangent=-0.0004, interpMode=RCIM_Cubic, leaveTangent=-0.0004, tangentMode=RCTM_User, time=35.1185, value=1}
      {interpMode=RCIM_Cubic, leaveTangent=0.01, time=60, value=1}
     ]}}
  }
 }
 {id=QualityLevel_0, children [Enveloper_3,Enveloper_3,Enveloper_2,Enveloper_2]}
 {
  id=WavePlayer_1
  properties {bLooping=true, soundWaveAssetPtr=&/Ambience_DCastlle_Courtyard_Morning_Ext_Loop_Quad}
 }
 {
  id=WavePlayer_3
  properties {bLooping=true, soundWaveAssetPtr=&/Ambience_DCastlle_Courtyard_Morning_Ext_Loop_Stereo}
 }
]
properties {
 attenuationSettings=&Fort_Ambient_Athena_Att
 maxDistance=2097152
 pitchMultiplier=1.5
 soundClassObject=&Fort_Ambient_Outdoors_Athena
 volumeMultiplier=0.165
}
